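# Physics asset for the StageCue gadget prop in the cosmetics animation content, giving its skeleton simple collision and joint setup.
# Capsule bodies sit on three bones: large ones on JNT_Table and root, and a tiny one on main.
# Two constraints chain JNT_Table to main and main to root, each with uniform swing and twist limit angles.
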PhysicsAsset /BRCosmetics/Animation/Props/Gadgets_StageCue/Meshes/StageCue_Gadget_Physics
bodies {
 JNT_Table [{shapeType=Capsule, length=25.468, radius=23.8239, center=<-0.0826 -0.1589 8.1179>, rotation=<rot 0 0.344 -88.6005>}]
 main [{shapeType=Capsule, length=0.505, radius=0.505, rotation=<rot 0 -90 0>}]
 root [{shapeType=Capsule, length=18.7206, radius=9.0151, center=<-7.9582 0 0.0175>, rotation=<rot 0 -90 -89.8959>}]
}
constraints [
 {bone1Name=main, bone2Name=root, swing1LimitDegrees=45, swing2LimitDegrees=45, twistLimitDegrees=45}
 {bone1Name=JNT_Table, bone2Name=main, swing1LimitDegrees=45, swing2LimitDegrees=45, twistLimitDegrees=45}
]
