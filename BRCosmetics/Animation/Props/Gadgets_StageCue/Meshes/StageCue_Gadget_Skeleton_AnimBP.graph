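# Animation Blueprint for the StageCue gadget prop that procedurally rolls its fifteen ball bones as animation moves them.
# Each update, RotateBall compares every ball's parent-space position with the stored one, deriving a roll axis perpendicular to XY travel and an angle from distance over BallRadius.
# The accumulated rotation is smoothed with RInterpTo and stored per ball.
# AnimGraph applies these rotations additively over the DefaultSlot pose through chained Modify Bone nodes.
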AnimBlueprint /BRCosmetics/Animation/Props/Gadgets_StageCue/Meshes/StageCue_Gadget_Skeleton_AnimBP
target_skeleton=&/StageCue_Gadget_Skeleton
graphs [
 AnimGraph {nodes [
   n1 = "Transform (Modify) Bone - Bone: JNT_Ball_01"(
    ComponentPose=Local To Component(LocalPose=Slot 'DefaultSlot'())
    Alpha=1
    Translation="0,0,0"
    Rotation="Get (a ref)"(Array:Array of Rotators=Get BallRotationsArray() -> (BallRotationsArray:Array of Rotators), Dimension 1=0) -> (Output:"Rotator (by ref)")
    Scale="1,1,1"
   ) {rotationMode=BMM_Additive, rotationSpace=BCS_ParentBoneSpace, alphaBoolBlend {blendOption=Linear}, boneToModify {boneName=JNT_Ball_01}}
   n2 = "Transform (Modify) Bone - Bone: JNT_Ball_02"(ComponentPose=@n1, Alpha=1, Translation="0,0,0", Rotation="Get (a ref)"(Array=Get BallRotationsArray(), Dimension 1=1), Scale="1,1,1") {rotationMode=BMM_Additive, rotationSpace=BCS_ParentBoneSpace, alphaBoolBlend {blendOption=Linear}, boneToModify {boneName=JNT_Ball_02}}
   n3 = "Transform (Modify) Bone - Bone: JNT_Ball_03"(ComponentPose=@n2, Alpha=1, Translation="0,0,0", Rotation="Get (a ref)"(Array=Get BallRotationsArray(), Dimension 1=2), Scale="1,1,1") {rotationMode=BMM_Additive, rotationSpace=BCS_ParentBoneSpace, alphaBoolBlend {blendOption=Linear}, boneToModify {boneName=JNT_Ball_03}}
   n4 = "Transform (Modify) Bone - Bone: JNT_Ball_04"(ComponentPose=@n3, Alpha=1, Translation="0,0,0", Rotation="Get (a ref)"(Array=Get BallRotationsArray(), Dimension 1=3), Scale="1,1,1") {rotationMode=BMM_Additive, rotationSpace=BCS_ParentBoneSpace, alphaBoolBlend {blendOption=Linear}, boneToModify {boneName=JNT_Ball_04}}
   n5 = "Transform (Modify) Bone - Bone: JNT_Ball_05"(ComponentPose=@n4, Alpha=1, Translation="0,0,0", Rotation="Get (a ref)"(Array=Get BallRotationsArray(), Dimension 1=4), Scale="1,1,1") {rotationMode=BMM_Additive, rotationSpace=BCS_ParentBoneSpace, alphaBoolBlend {blendOption=Linear}, boneToModify {boneName=JNT_Ball_05}}
   n6 = "Transform (Modify) Bone - Bone: JNT_Ball_06"(ComponentPose=@n5, Alpha=1, Translation="0,0,0", Rotation="Get (a ref)"(Array=Get BallRotationsArray(), Dimension 1=5), Scale="1,1,1") {rotationMode=BMM_Additive, rotationSpace=BCS_ParentBoneSpace, alphaBoolBlend {blendOption=Linear}, boneToModify {boneName=JNT_Ball_06}}
   n7 = "Transform (Modify) Bone - Bone: JNT_Ball_07"(ComponentPose=@n6, Alpha=1, Translation="0,0,0", Rotation="Get (a ref)"(Array=Get BallRotationsArray(), Dimension 1=6), Scale="1,1,1") {rotationMode=BMM_Additive, rotationSpace=BCS_ParentBoneSpace, alphaBoolBlend {blendOption=Linear}, boneToModify {boneName=JNT_Ball_07}}
   n8 = "Transform (Modify) Bone - Bone: JNT_Ball_08"(ComponentPose=@n7, Alpha=1, Translation="0,0,0", Rotation="Get (a ref)"(Array=Get BallRotationsArray(), Dimension 1=7), Scale="1,1,1") {rotationMode=BMM_Additive, rotationSpace=BCS_ParentBoneSpace, alphaBoolBlend {blendOption=Linear}, boneToModify {boneName=JNT_Ball_08}}
   n9 = "Transform (Modify) Bone - Bone: JNT_Ball_09"(ComponentPose=@n8, Alpha=1, Translation="0,0,0", Rotation="Get (a ref)"(Array=Get BallRotationsArray(), Dimension 1=8), Scale="1,1,1") {rotationMode=BMM_Additive, rotationSpace=BCS_ParentBoneSpace, alphaBoolBlend {blendOption=Linear}, boneToModify {boneName=JNT_Ball_09}}
   n10 = "Transform (Modify) Bone - Bone: JNT_Ball_10"(ComponentPose=@n9, Alpha=1, Translation="0,0,0", Rotation="Get (a ref)"(Array=Get BallRotationsArray(), Dimension 1=9), Scale="1,1,1") {rotationMode=BMM_Additive, rotationSpace=BCS_ParentBoneSpace, alphaBoolBlend {blendOption=Linear}, boneToModify {boneName=JNT_Ball_10}}
   n11 = "Transform (Modify) Bone - Bone: JNT_Ball_11"(ComponentPose=@n10, Alpha=1, Translation="0,0,0", Rotation="Get (a ref)"(Array=Get BallRotationsArray(), Dimension 1=10), Scale="1,1,1") {rotationMode=BMM_Additive, rotationSpace=BCS_ParentBoneSpace, alphaBoolBlend {blendOption=Linear}, boneToModify {boneName=JNT_Ball_11}}
   n12 = "Transform (Modify) Bone - Bone: JNT_Ball_12"(ComponentPose=@n11, Alpha=1, Translation="0,0,0", Rotation="Get (a ref)"(Array=Get BallRotationsArray(), Dimension 1=11), Scale="1,1,1") {rotationMode=BMM_Additive, rotationSpace=BCS_ParentBoneSpace, alphaBoolBlend {blendOption=Linear}, boneToModify {boneName=JNT_Ball_12}}
   n13 = "Transform (Modify) Bone - Bone: JNT_Ball_13"(ComponentPose=@n12, Alpha=1, Translation="0,0,0", Rotation="Get (a ref)"(Array=Get BallRotationsArray(), Dimension 1=12), Scale="1,1,1") {rotationMode=BMM_Additive, rotationSpace=BCS_ParentBoneSpace, alphaBoolBlend {blendOption=Linear}, boneToModify {boneName=JNT_Ball_13}}
   n14 = "Transform (Modify) Bone - Bone: JNT_Ball_14"(ComponentPose=@n13, Alpha=1, Translation="0,0,0", Rotation="Get (a ref)"(Array=Get BallRotationsArray(), Dimension 1=13), Scale="1,1,1") {rotationMode=BMM_Additive, rotationSpace=BCS_ParentBoneSpace, alphaBoolBlend {blendOption=Linear}, boneToModify {boneName=JNT_Ball_14}}
   n15 = "Transform (Modify) Bone - Bone: JNT_Ball_15"(ComponentPose=@n14, Alpha=1, Translation="0,0,0", Rotation="Get (a ref)"(Array=Get BallRotationsArray(), Dimension 1=14), Scale="1,1,1") {rotationMode=BMM_Additive, rotationSpace=BCS_ParentBoneSpace, alphaBoolBlend {blendOption=Linear}, boneToModify {boneName=JNT_Ball_15}}
   Output Pose(Result=Component To Local(ComponentPose=@n15))
  ]}
 EventGraph {nodes [
   n1 = Event BlueprintUpdateAnimation()
   > n2 = For Loop(LastIndex=14)
   n3 = Get BallRotationsArray()
   n4 = Get BallPositionsArray() -> (BallPositionsArray:Array of Vectors)
   > n5 = RotateBall!(
    execute
    self
    SkeletalMesh:Skeletal Mesh Component Object Reference=Get OwningComponent() -> (OwningComponent:Skeletal Mesh Component Object Reference)
    NewParam:Name="Get (a ref)"(Array:Array of Names=Get BallNamesArray() -> (BallNamesArray:Array of Names), Dimension 1=@n2) -> (Output:"Name (by ref)")
    DeltaTime:double=@n1.DeltaTimeX
    Position:Vector="Get (a ref)"(Array:Array of Vectors=@n4, Dimension 1=@n2) -> (Output:"Vector (by ref)")
    Rotation:Rotator="Get (a ref)"(Array:Array of Rotators=@n3, Dimension 1=@n2) -> (Output:"Rotator (by ref)")
   ) -> (then, NewBallRotation:Rotator, NewBallPosition:Vector)
   > Set Array Elem!(execute, TargetArray:Array of Rotators=@n3, Index:int=@n2, Item:"Rotator (by ref)"=@n5, bSizeToFit:bool=false) -> (then)
   > Set Array Elem(TargetArray:Array of Vectors=@n4, Index=@n2, Item:"Vector (by ref)"=@n5.NewBallPosition, bSizeToFit=false)
   Event BlueprintInitializeAnimation()
   > Set OwningComponent(OwningComponent:Skeletal Mesh Component Object Reference=GetOwningComponent()) -> (Output_Get:Skeletal Mesh Component Object Reference)
  ]}
 RotateBall {nodes [
   n1 = RotateBall!() -> (then, SkeletalMesh:Skeletal Mesh Component Object Reference, NewParam:Name, DeltaTime:double=~, Position:Vector=~, Rotation:Rotator=~) {is=title K2Node_FunctionEntry}
   n2 = GetSocketTransform!(self=@n1, InSocketName=@n1.NewParam, TransformSpace=RTS_ParentBoneSpace) -> (ReturnValue_Location:Vector, ReturnValue_Rotation:Rotator, ReturnValue_Scale:Vector)
   n3 = "Distance (Vector)"!(V1:Vector=@n2, V2:Vector=@n1.Position) -> (ReturnValue:double)
   n4 = Get BallRadius() -> (BallRadius:double)
   n5 = "Get Unit Direction (Vector)"!(From:Vector=@n1.Position, To:Vector=@n2) -> (ReturnValue_X:double, ReturnValue_Y:double, ReturnValue_Z:double) {is=title K2Node_CallFunction}
   n6 = RotatorFromAxisAndAngle!(
    Axis:Vector=MakeVector(X=float * float(A=@n5.ReturnValue_Y, B=-1) {is=title K2Node_PromotableOperator}, Y=@n5)
    Angle:float=Radians To Degrees!(A:double=float / float(A=@n3, B=@n4)) -> (ReturnValue:double)
   ) -> (ReturnValue:Rotator)
   > Return Node!(
    execute
    NewBallRotation:Rotator=RInterpTo!(
     Current:Rotator=@n1.Rotation
     Target:Rotator=Combine Rotators!(A:Rotator=@n1.Rotation, B:Rotator=@n6) -> (ReturnValue:Rotator)
     DeltaTime:float=@n1.DeltaTime
     InterpSpeed:float=5
    ) -> (ReturnValue:Rotator)
    NewBallPosition:Vector=@n2
   )
  ]}
]
variables [
 BallNamesArray {
  type="array<string>"
  default [JNT_Ball_01,JNT_Ball_02,JNT_Ball_03,JNT_Ball_04,JNT_Ball_05,JNT_Ball_06,JNT_Ball_07,JNT_Ball_08,JNT_Ball_09,JNT_Ball_10,JNT_Ball_11,JNT_Ball_12,JNT_Ball_13,JNT_Ball_14,JNT_Ball_15]
 }
 BallPositionsArray {
  type="array<Vector>"
  default [<0 0 0>, <0 0 0>, <0 0 0>, <0 0 0>, <0 0 0>, <0 0 0>, <0 0 0>, <0 0 0>, <0 0 0>, <0 0 0>, <0 0 0>, <0 0 0>, <0 0 0>, <0 0 0>, <0 0 0>]
 }
 BallRadius {type=number, default=0.5}
 BallRotationsArray {
  type="array<Rotator>"
  default [
   <rot 0 0 0>
   <rot 0 0 0>
   <rot 0 0 0>
   <rot 0 0 0>
   <rot 0 0 0>
   <rot 0 0 0>
   <rot 0 0 0>
   <rot 0 0 0>
   <rot 0 0 0>
   <rot 0 0 0>
   <rot 0 0 0>
   <rot 0 0 0>
   <rot 0 0 0>
   <rot 0 0 0>
   <rot 0 0 0>
  ]
 }
 OwningComponent {type=SkeletalMeshComponent}
]
